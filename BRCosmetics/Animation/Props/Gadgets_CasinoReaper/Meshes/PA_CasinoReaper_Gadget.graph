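# Physics asset for the CasinoReaper gadget cosmetic prop, a two-handed weapon rig with blade and card-stack bones.
# Each bone has one capsule, mostly tiny identical ones, apart from a larger handle capsule and a huge, near-spherical root capsule.
# Constraints follow one chain from root through main and handle to four serial blades, plus a branch from main through the suits joint to four card stacks.
# All joints share uniform swing and twist limit angles.
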PhysicsAsset /BRCosmetics/Animation/Props/Gadgets_CasinoReaper/Meshes/PA_CasinoReaper_Gadget
bodies {
 JNT_CardStack_01 [{shapeType=Capsule, length=0.505, radius=0.505, center=<0.007 0 0>, rotation=<rot 0 -89.0132 -90>}]
 JNT_CardStack_02 [{shapeType=Capsule, length=0.505, radius=0.505, center=<0.007 -0.0002 0>, rotation=<rot 0 89.0545 -90>}]
 JNT_CardStack_03 [{shapeType=Capsule, length=0.505, radius=0.505, center=<0.007 0.0003 0>, rotation=<rot 0 -86.8437 -90>}]
 JNT_CardStack_04 [{shapeType=Capsule, length=0.505, radius=0.505, center=<0.007 0.0002 0>, rotation=<rot 0 -87.0753 -90>}]
 JNT_Suits_MainMove [{shapeType=Capsule, length=0.505, radius=0.505, center=<0.007 -0.0002 -0.0001>, rotation=<rot 0 -88.469 -90>}]
 Melee_TwoHanded_Handle [{shapeType=Capsule, length=51.5141, radius=13.4676, center=<0 0.2868 37.3718>, rotation=<rot 0.687 -90 0>}]
 b_AR_Root [{shapeType=Capsule, length=0.505, radius=0.505, rotation=<rot 0 -90 0>}]
 blade_01 [{shapeType=Capsule, length=0.505, radius=0.505, center=<-0.1139 -0.7257 -0.5641>, rotation=<rot 0 0 -90>}]
 blade_02 [{shapeType=Capsule, length=0.505, radius=0.505, center=<-0.1629 -0.7853 -6.2514>, rotation=<rot 0 179.9999 -89.9999>}]
 blade_03 [{shapeType=Capsule, length=0.505, radius=0.505, center=<-0.1717 -0.7872 -5.7958>, rotation=<rot -0.0102 -90 0>}]
 blade_04 [{shapeType=Capsule, length=0.505, radius=0.505, center=<-0.1657 -0.6878 2.4644>, rotation=<rot 0.0475 -90 0.0006>}]
 main [{shapeType=Capsule, length=0.505, radius=0.505, rotation=<rot 0 -90 0>}]
 offset [{shapeType=Capsule, length=0.505, radius=0.505, rotation=<rot 0 -90 0>}]
 root [{shapeType=Capsule, length=2.8401, radius=340.1566, rotation=<rot 0 -90 0>}]
}
constraints [
 {bone1Name=offset, bone2Name=root, swing1LimitDegrees=45, swing2LimitDegrees=45, twistLimitDegrees=45}
 {bone1Name=main, bone2Name=offset, swing1LimitDegrees=45, swing2LimitDegrees=45, twistLimitDegrees=45}
 {bone1Name=b_AR_Root, bone2Name=main, swing1LimitDegrees=45, swing2LimitDegrees=45, twistLimitDegrees=45}
 {bone1Name=Melee_TwoHanded_Handle, bone2Name=b_AR_Root, swing1LimitDegrees=45, swing2LimitDegrees=45, twistLimitDegrees=45}
 {bone1Name=blade_01, bone2Name=Melee_TwoHanded_Handle, swing1LimitDegrees=45, swing2LimitDegrees=45, twistLimitDegrees=45}
 {bone1Name=blade_02, bone2Name=blade_01, swing1LimitDegrees=45, swing2LimitDegrees=45, twistLimitDegrees=45}
 {bone1Name=blade_03, bone2Name=blade_02, swing1LimitDegrees=45, swing2LimitDegrees=45, twistLimitDegrees=45}
 {bone1Name=blade_04, bone2Name=blade_03, swing1LimitDegrees=45, swing2LimitDegrees=45, twistLimitDegrees=45}
 {bone1Name=JNT_Suits_MainMove, bone2Name=main, swing1LimitDegrees=45, swing2LimitDegrees=45, twistLimitDegrees=45}
 {bone1Name=JNT_CardStack_01, bone2Name=JNT_Suits_MainMove, swing1LimitDegrees=45, swing2LimitDegrees=45, twistLimitDegrees=45}
 {bone1Name=JNT_CardStack_02, bone2Name=JNT_Suits_MainMove, swing1LimitDegrees=45, swing2LimitDegrees=45, twistLimitDegrees=45}
 {bone1Name=JNT_CardStack_03, bone2Name=JNT_Suits_MainMove, swing1LimitDegrees=45, swing2LimitDegrees=45, twistLimitDegrees=45}
 {bone1Name=JNT_CardStack_04, bone2Name=JNT_Suits_MainMove, swing1LimitDegrees=45, swing2LimitDegrees=45, twistLimitDegrees=45}
]
